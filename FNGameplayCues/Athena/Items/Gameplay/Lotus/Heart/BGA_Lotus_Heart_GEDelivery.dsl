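; Lotus Heart gameplay building actor, named as a GameplayEffect delivery device, that handles its beam, siren and death presentation.
; Custom Pulse runs off-server, attaching beam particles from the mesh targeting each player pawn overlapping the capsule, with debug name prints.
; BeamFX fires a local burst cue on each EffectedActors entry, ignoring its Actors parameter.
; BeginPlay fades out siren audio; death hides the actor server-side and plays a death cue.
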(event Building|EventOnBuildingActorInitialized
  (params (InitializationReason :type EFortBuildingInitializationReason%20Enum :default "StaticallyPlaced") (BuildingPersistentState :type EFortBuildingPersistentState%20Enum :default "Default"))
  (|CallParentFunction InitializationReason BuildingPersistentState))

(event Custom|Pulse
  (if (not (Networking|IsServer))
    (for _array_element (Collision|GetOverlappingActors (Variables|Default|GetCapsule) "/Script/FortniteGame.FortPlayerPawn" :!class /Script/Engine.PrimitiveComponent)
      (bind _returnvalue (Effects|SpawnEmitterAttached :EmitterTemplate "/FNGameplayCues/Athena/Items/Gameplay/Lotus/Heart/P_Lotus_Heart_Beam.P_Lotus_Heart_Beam" :AttachToComponent (Variables|Default|GetMesh) :Scale "1.000000,1.000000,1.000000" :LocationType "SnapToTarget" :bAutoDestroy true :PoolingMethod "AutoRelease" :bAutoActivate true))
      (Effects|Components|ParticleSystem|SetActorParameter _returnvalue "Target" _array_element)
      (Development|PrintString (Utilities|GetDisplayName _array_element)))))

(event EventBeginPlay
  (|CallParentFunction)
  (Audio|Components|Audio|FadeOut (Variables|Default|GetUrban_Siren_Evening_Mono_Cue) 9.0 0.0 "Linear" :!class /Script/Engine.AudioComponent))

(event Collision|EventActorBeginOverlap
  (params (OtherActor :type Actor%20Object%20Reference)))

(event EventTick
  (params (DeltaSeconds :type Float%20%28single-precision%29)))

(event Custom|BeamFX
  (params (Actors :type Array%20of%20Actor%20Object%20References))
  (for _array_element (Variables|Default|GetEffectedActors)
    (GameplayCue|ExecuteGameplayCue%28Burst%29Local _array_element (Variables|Default|GetBeamGC) (Ability|GameplayCue|MakeGameplayCueParameters :NormalizedMagnitude 0.0 :RawMagnitude 0.0 :Location (Transformation|GetActorLocation _array_element) :Normal "0,0,0" :SourceObject _array_element :GameplayEffectLevel 1 :AbilityLevel 1 :bReplicateLocationWhenUsingMinimalRepProxy false))))

(event EventOnDeathServer
  (params (Damage :type Float%20%28single-precision%29) (DamageTags :type Gameplay%20Tag%20Container%20Structure%20%28by%20ref%29) (Momentum :type Vector :default "0, 0, 0") (HitInfo :type Hit%20Result%20Structure%20%28by%20ref%29) (InstigatedBy :type Controller%20Object%20Reference) (DamageCauser :type Actor%20Object%20Reference) (EffectContext :type Gameplay%20Effect%20Context%20Handle%20Structure))
  (Rendering|SetActorHiddenInGame true))

(event EventOnDeathPlayEffects
  (params (Damage :type Float%20%28single-precision%29) (DamageTags :type Gameplay%20Tag%20Container%20Structure%20%28by%20ref%29) (Momentum :type Vector :default "0, 0, 0") (HitInfo :type Hit%20Result%20Structure%20%28by%20ref%29) (InstigatedBy :type Fort%20Pawn%20Object%20Reference) (DamageCauser :type Actor%20Object%20Reference) (EffectContext :type Gameplay%20Effect%20Context%20Handle%20Structure))
  (GameplayCue|ExecuteGameplayCue%28Burst%29Local self (Variables|Default|GetDeathGC) (Ability|GameplayCue|MakeGameplayCueParameters 0.0)))

(fn OnRep_EffectedActors)

(fn ConstructionScript)
