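; Gameplay ability for the Lotus Heart item that puts a banner on the avatar pawn while the ability is active.
; On activation it spawns BGA_Lotus_Heart_Banner at the avatar, snaps it to the pawn mesh's spine_03, and adds a gameplay cue to the owner.
; The cue's instigator comes from the granting effect's context.
; Ending the ability destroys the spawned banner; the failed passive activation event only calls its parent.
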(event Ability|EventActivateAbility
  (bind _returnvalue (Ability|GetAvatarActorfromActorInfo))
  (|CallParentFunction)
  (bind _returnvalue_2 (Building|SpawnBuildingGameplayActor "/FNGameplayCues/Athena/Items/Gameplay/Lotus/Heart/BGA_Lotus_Heart_Banner.BGA_Lotus_Heart_Banner_C" (Math|Conversions|ToTransform(Vector) (Transformation|GetActorLocation _returnvalue)) _returnvalue))
  (bind _asfort_player_pawn (Utilities|Casting|CastToFortPlayerPawn _returnvalue))
  (Transformation|AttachActorToComponent _returnvalue_2 (Variables|/Script/Engine.Character|GetMesh _asfort_player_pawn) "spine_03" "SnapToTarget" "KeepRelative" "KeepRelative" true)
  (Ability|AddGameplayCueWithParamsToOwner (Variables|Default|GetGC) (Ability|GameplayCue|MakeGameplayCueParameters :NormalizedMagnitude 0.0 :RawMagnitude 0.0 :Location "0,0,0" :Normal "0,0,0" :Instigator (Ability|EffectContext|GetInstigatorActor (Ability|GetGrantedbyEffectContext)) :EffectCauser (Ability|GetAvatarActorfromActorInfo) :GameplayEffectLevel 1 :AbilityLevel 1 :bReplicateLocationWhenUsingMinimalRepProxy false)))

(event Ability|EventFailedToActivatePassiveAbility
  (params (ActorInfo :type Gameplay%20Ability%20Actor%20Info%20Structure))
  (|CallParentFunction))

(event Ability|EventOnEndAbility
  (params (bWasCancelled :type Boolean))
  (Actor|DestroyActor (stale Ability|EventActivateAbility _returnvalue_2)))
