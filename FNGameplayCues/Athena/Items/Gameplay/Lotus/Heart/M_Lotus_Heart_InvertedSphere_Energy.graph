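# Additive, unlit, two-sided mesh-particle material with depth test disabled, giving an animated energy look for the Lotus Heart inverted-sphere effect.
# Two panning cloud samples, warped by each other and scene depth, blend with a normal-projected layer and are eroded by view angle.
# Manual scene-depth comparisons add fades and intersection edge glow, varying with camera distance.
# Particle color and time of day tint emissive; mobile substitutes fake depth.
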Material /FNGameplayCues/Athena/Items/Gameplay/Lotus/Heart/M_Lotus_Heart_InvertedSphere_Energy
graph {
 properties {
  bDisableDepthTest=true
  bUsedWithMeshParticles=true
  blendMode=BLEND_Additive
  pixelDepthOffsetMode=PDOM_Legacy
  shadingModel=MSM_Unlit
  twoSided=true
  usedShadingModels=Unlit
 }
 nodes [
  n1 = ComponentMask(None=VertexNormalWS()) {b=true}
  n2 = TextureObjectParameter() {parameterName=CloudTex, samplerType=SAMPLERTYPE_LinearGrayscale, texture=/Game/Packages/Fortress_Sky/SkyDome/Master/T_SkyDome_Cloud_Voroni}
  n3 = Multiply(A=TextureSample(UVs=Divide(A=@n1) {constB=256}, Tex=@n2) {samplerType=SAMPLERTYPE_LinearColor, texture=/Game/Packages/Fortress_Sky/SkyDome/Master/T_SkyDome_CloudShapes}.B) {constB=0.5}
  n4 = Multiply(A=TextureCoordinate(), B=DynamicParameter() {paramNames [TextureSize,Param2,Param3,Param4]}.TextureSize)
  n5 = Constant2Vector() {g=1.2, r=3}
  n6 = TextureSample(UVs=Panner(Coordinate=Multiply(A=@n4, B=@n5)) {speedX=0.01, speedY=0.1}, Tex=@n2) {samplerType=SAMPLERTYPE_LinearColor, texture=&T_SkyDome_CloudShapes}
  n7 = Multiply(A=@n6.B) {constB=0.1}
  n8 = SceneDepth()
  n9 = ShadingPathSwitch(Default=@n8.R, Deferred=@n8.R, Mobile=ScalarParameter() {defaultValue=750, parameterName=FakeDepthForLow}) {desc=Migrated from FeatureLevelSwitch}
  n10 = Divide(A=Saturate(None=Divide(A=@n9) {constB=1024})) {constB=32}
  n11 = Panner(Coordinate=Multiply(A=Add(A=@n10, B=@n4), B=Constant2Vector() {g=0.5, r=1})) {speedX=-0.01, speedY=0.1}
  n12 = Multiply(
   A=@n6.B
   B=TextureSample(UVs=Add(A=@n7, B=@n11), Tex=@n2) {samplerType=SAMPLERTYPE_LinearColor, texture=&T_SkyDome_CloudShapes}.B
  )
  n13 = Sine(None=TextureCoordinate() {uTiling=0.5, vTiling=0.5})
  n14 = Saturate(None=ComponentMask(None=Subtract(A=@n13) {constB=0.2}) {g=true})
  n15 = SquareRoot(None=Add(A=LinearInterpolate(A=@n3, B=@n12, Alpha=@n14)) {constB=0.2})
  n16 = Abs(None=DotProduct(A=VertexNormalWS(), B=CameraVectorWS()))
  n17 = SphereMask(A=WorldPosition().XYZ, B=CameraPositionWS(), Radius=Constant() {r=512}) {hardnessPercent=0}
  n18 = OneMinus(None=Multiply(A=SquareRoot(None=@n16), B=OneMinus(None=@n17)))
  n19 = Saturate(None=Divide(A=@n9, B=ScalarParameter() {defaultValue=256, parameterName=CloseFade}))
  n20 = Distance(A=WorldPosition().XYZ, B=CameraPositionWS())
  n21 = ScalarParameter() {defaultValue=256, parameterName=EdgeDistance}
  n22 = Divide(A=Subtract(A=@n9, B=@n20), B=@n21)
  n23 = Distance(A=ObjectPositionWS(), B=CameraPositionWS())
  n24 = ScalarParameter() {defaultValue=6144, parameterName=ShowInside}
  n25 = Saturate(None=Add(A=Divide(A=@n23, B=@n24) {constB=4096}, B=ScalarParameter() {defaultValue=-0.5, parameterName=InsideShowsUpBIse}) {constA=0.25, constB=-0.25})
  n26 = SphereMask(A=@n20, B=@n9, Radius=@n21) {hardnessPercent=0}
  n27 = Multiply(A=@n26, B=@n26)
  n28 = Multiply(
   A=Saturate(None=LinearInterpolate(A=@n19, B=@n22, Alpha=@n25) {constA=1})
   B=LinearInterpolate(
    A=ScalarParameter() {parameterName=NoneEdgeEnhance}
    B=ScalarParameter() {defaultValue=1.4, parameterName=EdgeEnhance}
    Alpha=Saturate(None=@n27)
   ) {constA=1}
  )
  n29 = Divide(A=Multiply(A=Subtract(A=@n15, B=@n18), B=Clamp(None=@n28) {clampMode=CMODE_ClampMax, maxDefault=4}) {constA=1})
  n30 = ParticleColor()
  MaterialFunctionCall(Emissive=Multiply(
    A=MaterialFunctionCall() {materialFunction=&Color_Time_of_Day}.Result
    B=Multiply(A=@n29, B=@n30.RGB)
   )) {materialFunction=&MF_FX_EmissiveAdjustments}
  n31 = CurveAtlasRowParameter(CurveTime=@n29) {atlas=&AthenaCurveAtlas01, curve=/Game/Athena/Items/Consumables/AirStrike/Curve_RedSmoke, defaultValue=110}
  Multiply(A=@n30.A, B=@n31.A)
  SceneDepth()
  SceneDepth()
  TextureSample() {samplerType=SAMPLERTYPE_LinearGrayscale, texture=&T_SkyDome_Cloud_Voroni}
 ]
}
